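; Construction script for PirateShip_Mast_S, a building prop Blueprint that picks an elemental symbol material from IceCastle chair assets, despite its mast name.
; It switches on the ChairNumber integer to set material slot 1 of its StaticMeshComponent.
; Values 1 through 5 map to Earth, Fire, Ice, Water and Wind symbols; other values leave the material unchanged.
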(fn ConstructionScript
  (bind _staticmeshcomponent (Variables|Debug|GetStaticMeshComponent))
  (switch int (Variables|Default|GetChairNumber)
    (:1
      (Rendering|Material|SetMaterial _staticmeshcomponent 1 "/Game/Environments/Sets/IceCastle/Materials/MI_IceCastle_Chair_Symbol_Earth.MI_IceCastle_Chair_Symbol_Earth"))
    (:2
      (Rendering|Material|SetMaterial _staticmeshcomponent 1 "/Game/Environments/Sets/IceCastle/Materials/MI_IceCastle_Chair_Symbol_Fire.MI_IceCastle_Chair_Symbol_Fire"))
    (:3
      (Rendering|Material|SetMaterial _staticmeshcomponent 1 "/Game/Environments/Sets/IceCastle/Materials/MI_IceCastle_Chair_Symbol_Ice.MI_IceCastle_Chair_Symbol_Ice"))
    (:4
      (Rendering|Material|SetMaterial _staticmeshcomponent 1 "/Game/Environments/Sets/IceCastle/Materials/MI_IceCastle_Chair_Symbol_Water.MI_IceCastle_Chair_Symbol_Water"))
    (:5
      (Rendering|Material|SetMaterial _staticmeshcomponent 1 "/Game/Environments/Sets/IceCastle/Materials/MI_IceCastle_Chair_Symbol_Wind.MI_IceCastle_Chair_Symbol_Wind"))))
